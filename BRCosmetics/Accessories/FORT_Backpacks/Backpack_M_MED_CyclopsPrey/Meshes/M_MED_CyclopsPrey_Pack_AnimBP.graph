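# Animation Blueprint for the CyclopsPrey backpack that layers procedural motion over a looping idle.
# Each update maps the owner mesh's spine bend to additive roll and offset on the pack bone, on top of a small fixed nudge.
# A real-time sine wave bobs dyn_backpack_base vertically.
# Head-down, melee and pickaxe arm values are also calculated, but the AnimGraph does not use them.
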AnimBlueprint /BRCosmetics/Accessories/FORT_Backpacks/Backpack_M_MED_CyclopsPrey/Meshes/M_MED_CyclopsPrey_Pack_AnimBP
target_skeleton=&/M_MED_CyclopsPrey_Pack_Skeleton
graphs [
 AnimGraph {nodes [
   Input Pose()
   AnimGraphNode_StateMachine_0 = Reactive() {is=title AnimGraphNode_StateMachine}
   n1 = "Transform (Modify) Bone - Bone: pack"(ComponentPose=Local To Component(LocalPose=@AnimGraphNode_StateMachine_0), Translation="0,0.5,0", Rotation="0,0,0") {
    translationMode=BMM_Additive
    translationSpace=BCS_BoneSpace
    alphaBoolBlend {blendOption=Linear}
    boneToModify {boneName=pack}
    translation {y=0.5}
   }
   n2 = "Transform (Modify) Bone - Bone: pack"(
    ComponentPose=@n1
    Translation=Get Backpack Translation() -> (Backpack Translation:Vector)
    Rotation=Get Backpack Rotate() -> (Backpack Rotate:Rotator)
   ) {
    rotationMode=BMM_Additive
    rotationSpace=BCS_BoneSpace
    translationMode=BMM_Additive
    translationSpace=BCS_BoneSpace
    alphaBoolBlend {blendOption=Linear}
    boneToModify {boneName=pack}
   }
   n3 = "Transform (Modify) Bone - Bone: dyn_backpack_base"(ComponentPose=@n2, Alpha=1, Translation=Get Translate Main() -> (Translate Main:Vector), Rotation="0,0,0", Scale="1,1,1") {translationMode=BMM_Additive, translationSpace=BCS_BoneSpace, alphaBoolBlend {blendOption=Linear}, boneToModify {boneName=dyn_backpack_base}}
   Output Pose(Result=Component To Local(ComponentPose=@n3)) {layerGroup=None}
  ]}
 Reactive {entry=Idle, path=AnimGraph.AnimGraphNode_StateMachine_0.Reactive, ids=id, nodes [AnimStateNode_0 = Idle()]}
 Idle {
  path=AnimGraph.AnimGraphNode_StateMachine_0.Reactive.AnimStateNode_0.Idle
  nodes [Output Animation Pose(Result=Sequence Player 'CyclopsPrey_IDLE'() {sequence=/BRCosmetics/Accessories/FORT_Backpacks/Backpack_M_MED_CyclopsPrey/Anim/CyclopsPrey_IDLE})]
 }
 EventGraph {nodes [
   Event BlueprintUpdateAnimation()
   > n1 = Sequence!(execute) -> (then_0, then_1, then_2:Exec)
   n2 = GetRealTimeSeconds()
   n3 = "Sin (Radians)"(A=float * float(A=float * float(A=1.65, B=-1.65), B=@n2) {is=title K2Node_PromotableOperator})
   Set Translate Main(execute=@n1.then_1, Translate Main:Vector=MakeVector(Y=0, Z=float / float(A=@n3, B=1.3))) -> (Output_Get:Vector)
   n4 = GetOwnerSkeletalMeshComponent()
   n5 = BreakRotator(InRot=BreakTransform(InTransform=GetDeltaTransformFromRefPose(self=@n4, BoneName=spine_05, BaseName=spine_01)).Rotation)
   Set Backpack Rotate(execute=@n1.then_0, Backpack Rotate:Rotator=MakeRotator(Roll=MapRangeClamped(Value=@n5.Yaw, InRangeA=0, InRangeB=-40, OutRangeB=30))) -> (Output_Get:Rotator)
   > Set Backpack Translation(Backpack Translation:Vector=MakeVector(Y=MapRangeClamped(Value=@n5.Yaw, InRangeA=0, InRangeB=-30, OutRangeB=-5))) -> (Output_Get:Vector)
   n6 = GetOwnerSkeletalMeshComponent()
   > n7 = Set head down rot(head down rot:double=GetAlphaFromReferenceTransformDelta(SkeletalMeshComponent=@n6, BoneName=spine_01, BaseName=head, InputRange=(X=0,Y=-45), TwistAxis="1,0,0")) -> (Output_Get:double)
   n8 = "To Float (Boolean)"(InBool=Get bIsPlayingMeleeAnim())
   n9 = float * float(B=0.2)
   n10 = Get isMelee() -> (isMelee:double)
   > Set isMelee(isMelee:double=float * float(A=@n7, B=FInterpTo(Current=@n10, Target=@n8, DeltaTime=@n9, InterpSpeed=5))) -> (Output_Get:double)
   n11 = BreakRotator(InRot=BreakTransform(InTransform=GetDeltaTransformFromRefPose(self=@n6, BoneName=upperarm_l, BaseName=clavicle_l)).Rotation)
   > Set Backpack Rotate Pickaxe(Backpack Rotate Pickaxe:Rotator=MakeRotator(Roll=MapRangeClamped(Value=@n11.Yaw, InRangeA=0, InRangeB=90, OutRangeB=30))) -> (Output_Get:Rotator)
   n12 = float * float(A=0.8, B=-0.8)
   float / float(A="Sin (Radians)"(A=float * float(A=@n12, B=@n2)), B=0.5)
   n13 = float * float(A=1, B=1.3)
   float / float(A="Sin (Radians)"(A=float * float(A=@n13, B=@n2)), B=1)
  ]}
]
variables [
 Backpack Rotate {type=Rotator}
 Backpack Rotate Pickaxe {type=Rotator}
 Backpack Translation {type=Vector}
 head down rot {type=number}
 isMelee {type=number}
 RB Alpha {type=number}
 Translate Main {type=Vector}
]
